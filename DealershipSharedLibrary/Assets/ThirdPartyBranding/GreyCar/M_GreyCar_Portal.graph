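# Additive, unlit, two-sided portal effect material for Niagara mesh particles, in the GreyCar third-party branding assets.
# Local-position curl noise, scrolled by particle Time and distorted by rotating, panning noise, breaks up a mirrored-UV sphere mask.
# MF_DLR_Erode erodes that mask through dynamic parameters, feeding HDR-remapped particle colour and an alpha shaped by depth fade and two-sided Fresnel.
# Side branches compute camera-distance fades and a bounds-weighted inertia vector.
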Material /DealershipSharedLibrary/Assets/ThirdPartyBranding/GreyCar/M_GreyCar_Portal
graph {
 properties {
  bAutomaticallySetUsageInEditor=false
  bUsedWithNiagaraMeshParticles=true
  blendMode=BLEND_Additive
  shadingModel=MSM_Unlit
  twoSided=true
  usedShadingModels=Unlit
 }
 nodes [
  n1 = TextureCoordinate()
  n2 = ComponentMask(None=@n1) {r=true}
  n3 = OneMinus(None=@n2)
  n4 = Add(A=Multiply(A=Min(A=@n2, B=@n3)) {constB=2}) {constB=0.25}
  n5 = ComponentMask(None=@n1) {g=true}
  n6 = Min(A=@n5, B=OneMinus(None=@n5))
  n7 = OneMinus(None=AppendVector(A=Saturate(None=@n4), B=Multiply(A=@n6) {constB=2}))
  n8 = Distance(A=@n7, B=Constant())
  LinearInterpolate(Alpha=Power(Base=OneMinus(None=@n8), Exp=ScalarParameter() {defaultValue=32})) {constB=0}
  n9 = ParticleColor()
  n10 = DynamicParameter() {paramNames [Noise 1, Noise 2, Erosion, Erosion Width]}
  n11 = Add(
   A=Multiply(A=LocalPosition().XYZ, B=VectorParameter() {parameterName=Param_1, defaultValue=<rgba 0.003 0.007 0.005 1>}.RGB) {constB=0.01}
   B=Multiply(
    A=DynamicParameter() {parameterIndex=1, paramNames [Time, Param2, Param3, Scale Y]}.Time
    B=VectorParameter() {defaultValue=<rgba -1.2 0 0.1 1>}.RGB
   )
  )
  n12 = Panner(Coordinate=Rotator(Coordinate=TextureCoordinate()) {speed=0.2}) {speedY=0.2}
  n13 = Multiply(
   A=Subtract(A=TextureSample(UVs=@n12) {samplerType=SAMPLERTYPE_LinearColor, texture=&HighContrastNoise}.RGB) {constB=0.5}
   B=DynamicParameter() {paramNames [Noise 1, Noise 2, Erosion, Erosion Width]}.Noise 2
  )
  n14 = ComponentMask(None=TextureSample(UVs=Add(A=@n11, B=@n13)) {samplerType=SAMPLERTYPE_LinearColor, texture=/Game/Effects/FogVolume/VT_CurlNoise_LOW}.RGB) {g=true, r=true}
  n15 = DynamicParameter() {paramNames [Noise 1, Noise 2, Erosion, Erosion Width]}
  n16 = MaterialFunctionCall!(
   Erosion=@n10.Erosion
   Mask=SphereMask(A=@n7, B=Multiply(A=@n14, B=@n15.Noise 1), Radius=Constant() {r=0.75}) {hardnessPercent=0}
   Edge Width
   Width=@n10.Erosion
  ) {materialFunction=/DealershipSharedLibrary/FX/Materials/MaterialFunctions/MF_DLR_Erode}
  n17 = Add(
   A=@n16.Result
   B=MaterialFunctionCall(In=@n16.Result, Contrast=Constant() {r=1}) {materialFunction=&CheapContrast}.Result
  )
  n18 = Multiply(A=@n9.RGB, B=@n17)
  MaterialFunctionCall(Emissive=MaterialFunctionCall(LDR Input=Saturate(None=@n18), HDR Tint=@n18) {materialFunction=&DeriveHDRfromLDR}.RemappedRange) {materialFunction=&MF_FX_EmissiveAdjustments}
  n19 = ScalarParameter() {defaultValue=1000, parameterName=Fade distance}
  Multiply(A=MaterialFunctionCall(Fade Offset=@n19) {materialFunction=&CameraDepthFade}.Result)
  n20 = LinearInterpolate(
   A=ScalarParameter() {defaultValue=1000, parameterName=DepthFade Distance - Max}
   B=ScalarParameter() {defaultValue=400, parameterName=DepthFade Distance - Min}
   Alpha=@n16.Result
  )
  n21 = ScalarParameter() {defaultValue=1.5, parameterName=Param_1}
  n22 = ScalarParameter() {defaultValue=-0.3, parameterName=Param_4}
  n23 = Fresnel(
   ExponentIn=LinearInterpolate(
    A=ScalarParameter() {defaultValue=0.3, parameterName=Param_2}
    B=ScalarParameter() {defaultValue=0.7, parameterName=ExponentIn}
    Alpha=@n16.Result
   ) {constA=2}
   Normal=Multiply(A=VertexNormalWS(), B=TwoSidedSign())
  ) {baseReflectFraction=0}
  Multiply(
   A=@n9.A
   B=Multiply(
    A=MaterialFunctionCall(Opacity In=@n17, DepthFade Distance=@n20) {materialFunction=&MF_VCVFX_DepthFade}.Result
    B=LinearInterpolate(A=@n21, B=@n22, Alpha=@n23) {constA=1, constB=0}
   )
  )
  n24 = LocalPosition()
  n25 = ObjectLocalBounds()
  n26 = OneMinus(None=Divide(A=Add(A=@n24.XYZ, B=@n25.Half Extents), B=@n25.Extents))
  Multiply(
   A=VectorParameter() {group=Inertia, parameterName=Dynamic Inertia - Position, defaultValue=<rgba 0 0 0 1>}.RGB
   B=Power(
    Base=ComponentMask(None=@n26) {r=true}
    Exp=ScalarParameter() {defaultValue=54, group=Inertia, parameterName=Dynamic Inertia - Exponent, sortPriority=2}
   ) {constExponent=3}
  )
  OneMinus(None=SphereMask(A=CameraPositionWS(), B=Constant(), Radius=@n19, Hardness=ScalarParameter() {defaultValue=0.8, parameterName=Fade Falloff}))
  ScalarParameter() {parameterName=Param_3}
  WorldPosition()
 ]
}
